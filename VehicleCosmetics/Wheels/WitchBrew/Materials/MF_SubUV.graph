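# MF_SubUV is a flipbook material function in the WitchBrew wheel cosmetic, sampling one cell of a texture grid from Frames, SubImages and UVs.
# By default wiring, it floors Frames into column and row offsets over UVs scaled by 1/SubImages.
# The LinearBlend static switch instead lerps current and next-frame samples by the fractional frame, for RGB and Alpha.
# That blend path scales by SubImages divided by one, not its reciprocal.
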MaterialFunction /VehicleCosmetics/Wheels/WitchBrew/Materials/MF_SubUV
graph {nodes [
  n1 = FunctionInput() {inputName=Frames, inputType=FunctionInput_Scalar}
  n2 = Frac(None=@n1)
  n3 = Subtract(A=@n1, B=@n2)
  n4 = FunctionInput() {inputName=SubImages, inputType=FunctionInput_Vector2, previewValue=<1 1 0 1>}
  n5 = ComponentMask(None=@n4) {r=true}
  n6 = Divide(A=@n4, B=Constant2Vector() {g=1, r=1})
  n7 = ComponentMask(None=@n6) {r=true}
  n8 = AppendVector(A=Fmod(A=@n3, B=@n5), B=Floor(None=Multiply(A=@n3, B=@n7)))
  n9 = FunctionInput(Preview=TextureCoordinate()) {bUsePreviewValueAsDefault=true, inputName=UVs, inputType=FunctionInput_Vector2}
  n10 = FunctionInput(Preview=TextureObject()) {bUsePreviewValueAsDefault=true, inputName=Texture, inputType=FunctionInput_Texture2D}
  n11 = TextureSample(UVs=Multiply(A=Add(A=@n9, B=@n8), B=@n6), Tex=@n10)
  n12 = Add(A=@n3, B=Constant() {r=1})
  n13 = AppendVector(A=Fmod(A=@n12, B=@n5), B=Floor(None=Multiply(A=@n7, B=@n12)))
  n14 = TextureSample(UVs=Multiply(A=Add(A=@n13, B=@n9), B=@n6), Tex=@n10)
  n15 = Constant() {r=1}
  n16 = ComponentMask(None=@n4) {r=true}
  n17 = ComponentMask(None=@n4) {g=true}
  n18 = AppendVector(A=Divide(A=@n15, B=@n16), B=Divide(A=@n15, B=@n17))
  n19 = Divide(A=Fmod(A=Floor(None=@n1), B=@n16), B=@n16)
  n20 = Divide(A=Floor(None=Divide(A=@n1, B=@n16)), B=@n17)
  n21 = TextureSample(UVs=Add(A=Multiply(A=@n18, B=@n9), B=AppendVector(A=@n19, B=@n20)), Tex=@n10) {texture=/VehicleCosmetics/Wheels/WitchBrew/Textures/T_WitchBrew_Rim_Glow}
  n22 = FunctionInput(Preview=StaticBool()) {bUsePreviewValueAsDefault=true, inputName="LinearBlend(MoreExpensive)", inputType=FunctionInput_StaticBool}
  FunctionOutput(None=StaticSwitch(True=LinearInterpolate(A=@n11.RGB, B=@n14.RGB, Alpha=@n2), False=@n21.RGB, Value=@n22)) {outputName=RGB}
  FunctionOutput(None=StaticSwitch(True=LinearInterpolate(A=@n11.A, B=@n14.A, Alpha=@n2), False=@n21.A, Value=@n22)) {outputName=Alpha}
 ]}
