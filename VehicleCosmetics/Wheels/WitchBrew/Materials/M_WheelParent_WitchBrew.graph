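# Masked parent material for the WitchBrew vehicle wheel, rendering an animated bubbling brew on skeletal meshes.
# Speed drives 8x4 flipbooks for base, bubble, color mask and normal textures, with frame-blended bubble alpha feeding a BumpOffset for parallax.
# CustomColor tints masked areas, combined with an MF_WordENV environment term and a boosted glow confined by a radial UV mask.
# Crossed time-driven sines along vertex normals, limited by that mask, form a wobble offset.
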Material /VehicleCosmetics/Wheels/WitchBrew/Materials/M_WheelParent_WitchBrew
graph {
 properties {
  bAutomaticallySetUsageInEditor=false
  bUsedWithSkeletalMesh=true
  blendMode=BLEND_Masked
  previewMesh=/VehicleCosmetics/Wheels/Twist_Twist/Meshes/SK_Wheel_Twist_Twist
  usedShadingModels=Default Lit
 }
 nodes [
  n1 = Constant2Vector() {g=4, r=8}
  n2 = Divide(A=Constant2Vector() {g=1, r=1}, B=@n1)
  n3 = ComponentMask(None=@n2) {r=true}
  n4 = Multiply(A=Time(), B=ScalarParameter() {defaultValue=12, parameterName=Speed})
  n5 = Subtract(A=@n4, B=Frac(None=@n4))
  n6 = ComponentMask(None=@n1) {r=true}
  n7 = AppendVector(A=Floor(None=Multiply(A=@n3, B=@n5)), B=Fmod(A=@n5, B=@n6))
  n8 = TextureCoordinate()
  n9 = TextureSample(UVs=Multiply(A=Add(A=@n7, B=@n8), B=@n2)) {samplerType=SAMPLERTYPE_LinearColor, texture=/VehicleCosmetics/Wheels/WitchBrew/Textures/T_WitchBrew_SE_Bubbles_BC}
  n10 = Add(A=@n5, B=Constant() {r=1})
  n11 = AppendVector(A=Floor(None=Multiply(A=@n3, B=@n10)), B=Fmod(A=@n10, B=@n6))
  n12 = TextureSample(UVs=Multiply(A=Add(A=@n11, B=@n8), B=@n2)) {samplerType=SAMPLERTYPE_LinearColor, texture=&T_WitchBrew_SE_Bubbles_BC}
  n13 = Frac(None=@n4)
  n14 = BumpOffset(Coordinate=TextureCoordinate(), Height=Multiply(A=LinearInterpolate(A=@n9.A, B=@n12.A, Alpha=@n13), B=Constant() {r=2}))
  n15 = Constant2Vector() {g=4, r=8}
  n16 = StaticBool()
  n17 = MaterialFunctionCall!(
   UVs=@n14
   Texture=TextureObject() {texture=/VehicleCosmetics/Wheels/WitchBrew/Textures/WitchBrew_Bubbles_D}
   SubImages=@n15
   "LinearBlend(MoreExpensive)"=@n16
   Frames=@n4
  ) {materialFunction=&/MF_SubUV}
  n18 = Multiply(A=@n17.Alpha, B=Constant() {r=0.5})
  n19 = OneMinus(None=ConstantBiasScale(None=@n18) {bias=-0.03, scale=30})
  n20 = Multiply(A=@n19, B=Constant() {r=0.15})
  n21 = LinearInterpolate(
   A=MaterialFunctionCall(
    UVs=@n14
    Texture=TextureObject() {texture=/VehicleCosmetics/Wheels/WitchBrew/Textures/WitchBrew_Base_D}
    SubImages=@n15
    "LinearBlend(MoreExpensive)"=@n16
    Frames=@n4
   ) {materialFunction=&MF_SubUV}.RGB
   B=@n17.RGB
   Alpha=@n20
  )
  n22 = VectorParameter() {parameterName=CustomColor, defaultValue=<rgba 0.0999 1 0.1195 1>}
  n23 = Multiply(A=@n22.RGB, B=Constant() {r=0.57})
  n24 = Desaturation(None=@n22.RGB, Fraction=Constant() {r=0.35}) {luminanceFactors=<rgba 0.2918 0.5776 0.1095 1>}
  n25 = Clamp(None=@n19)
  n26 = LinearInterpolate(
   A=@n21
   B=Multiply(A=LinearInterpolate(A=@n23, B=@n24, Alpha=@n25), B=@n21)
   Alpha=MaterialFunctionCall(
    UVs=@n14
    Texture=TextureObject() {texture=/VehicleCosmetics/Wheels/WitchBrew/Textures/WitchBrew_ColorMask}
    SubImages=@n15
    "LinearBlend(MoreExpensive)"=@n16
    Frames=@n4
   ) {materialFunction=&MF_SubUV}.RGB
  )
  n27 = Constant() {r=0.25}
  n28 = MaterialFunctionCall!(Tint=@n26, TintAmount=Constant() {r=1}, Desaturate=Constant()) {materialFunction=/VehicleCosmetics/Wheels/Oldmans/New/Materials/MF_WordENV}
  n29 = Multiply(A=@n28.LocalSpace, B=Constant() {r=0.5})
  n30 = Multiply(A=@n26, B=Constant() {r=5})
  n31 = Constant()
  n32 = Add(A=TextureCoordinate() {uTiling=2, vTiling=2}, B=Constant() {r=-1})
  n33 = DotProduct(A=@n32, B=@n32)
  n34 = Constant() {r=0.75}
  n35 = Clamp(None=Multiply(A=Subtract(A=@n33, B=@n34), B=Constant() {r=2}))
  Add(
   A=Add(A=Multiply(A=@n26, B=@n27), B=LinearInterpolate(A=@n29, B=@n28.LocalSpace, Alpha=@n20))
   B=Multiply(A=LinearInterpolate(A=@n30, B=@n31, Alpha=@n25), B=@n35)
  )
  n36 = Multiply(A=@n18, B=Constant() {r=2})
  Add(A=@n36, B=@n19)
  n37 = Constant() {r=5}
  Clamp(None=OneMinus(None=Multiply(A=@n35, B=@n37)))
  LinearInterpolate()
  n38 = Multiply(A=Time(), B=Constant() {r=2})
  n39 = Add(A=ComponentMask(None=TextureCoordinate() {uTiling=8, vTiling=8}) {g=true}, B=@n38)
  n40 = TextureCoordinate() {uTiling=8, vTiling=512}
  n41 = Add(A=@n38, B=OneMinus(None=ComponentMask(None=@n40) {r=true}))
  n42 = ConstantBiasScale(None=Multiply(A=Sine(None=@n39) {period=2}, B=Sine(None=@n41) {period=2}))
  n43 = VertexNormalWS()
  Multiply(A=Multiply(A=Multiply(A=@n42, B=@n43), B=Constant() {r=2}), B=@n35)
  Multiply(A=Max(A=@n36, B=@n19), B=ScalarParameter() {defaultValue=1, parameterName=Roughness})
  n44 = TextureObject() {samplerType=SAMPLERTYPE_Normal, texture=/VehicleCosmetics/Wheels/WitchBrew/Textures/WitchBrew_Base_N}
  n45 = MaterialFunctionCall(
   UVs=@n14
   Texture=TextureObject() {samplerType=SAMPLERTYPE_Normal, texture=/VehicleCosmetics/Wheels/WitchBrew/Textures/WitchBrew_Bubbles_N}
   SubImages=@n15
   "LinearBlend(MoreExpensive)"=@n16
   Frames=@n4
  ) {materialFunction=&MF_SubUV}
  n46 = Constant() {r=0.5}
  Multiply(
   A=LinearInterpolate(
    A=MaterialFunctionCall(UVs=@n14, Texture=@n44, SubImages=@n15, "LinearBlend(MoreExpensive)"=@n16, Frames=@n4) {materialFunction=&MF_SubUV}.RGB
    B=@n45.Alpha
    Alpha=Multiply(A=@n45.Alpha, B=@n46)
   )
   B=Constant3Vector() {constant=<rgba 2 2 1 0>}
  )
  PixelNormalWS()
  ScalarParameter() {parameterName=Metallic}
 ]
}
